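# Niagara idle effect for the Drift one-handed pickaxe, emitting short-lived glowing sprites from a thin, tall box.
# One emitter runs two spawn groups via masked inputs: one drifts along negative Y with an initial brightness flash, the other stays put and grows.
# Spawn group sets the visibility tag for two sprite renderers alongside a light renderer.
# User.VariantColor tints particles, while dynamic material parameters add randomized and over-life variation.
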NiagaraSystem /BRCosmetics/Weapons/FORT_Melee/Pickaxe_Drift1H/FX/NS_Drift1H_Pickaxe_Idle
dependencies {usedDataInterfaces [&NiagaraDataInterfaceCurve], usedDynamicInputs [], usedModules [&SystemState]}
emitters {Idle {
  emitterUpdateScript [
   &EmitterState()
   &SpawnRate(SpawnRate=3, Use Spawn Probability=false)
   &SpawnRate(SpawnGroup=1, SpawnRate=3, Use Spawn Probability=false) {module=SpawnRate001}
  ]
  particleSpawnScript [
   /SetVariables(Particles.VisibilityTag=&SpawnGroup()) {isSetParameters=true}
   &InitializeParticle(
    Color=&MaskLinearColorBySpawnGroup(
     Masked Linear Color=<rgba 4 4 4 1>
     Passthrough Linear Color=&RandomRangeLinearColor(Minimum=<rgba 7 7 7 1>, Maximum=<rgba 8 8 8 1>)
    )
    Lifetime Max=&MaskFloatBySpawnGroup(Masked Float=0.3, Passthrough Float=0.25)
    Lifetime Min:NiagaraFloat=&MaskFloatBySpawnGroup(Masked Float=0.2, Passthrough Float=0.15)
    Lifetime Mode=::NewEnumerator1
    Sprite Rotation Angle=&MaskFloatBySpawnGroup(Masked Float=0, Passthrough Float=&RandomRangeFloat(Minimum=0, Maximum=1))
    Sprite Rotation Mode=::NewEnumerator1
    Sprite Size Max=&MaskVector2dBySpawnGroup(Masked Vector2D=<35 35>, Passthrough Vector2D=<25 35>, Spawn Group Mask=0)
    Sprite Size Min=&MaskVector2dBySpawnGroup(Masked Vector2D=<25 25>, Passthrough Vector2D=<15 30>, Spawn Group Mask=0)
    Sprite Size Mode=::NewEnumerator2
    UsePositionOffset=false
    Write Material Random=true
   )
   &CameraOffset(Camera Offset Amount=&MaskFloatBySpawnGroup(Masked Float=5, Passthrough Float=15))
   &ShapeLocation(
    Box Midpoint=<0 0.5 0.4>
    Box Size=<0 5 40>
    MaskSpawnGroup=false
    Override Seed=false
    Shape Primitive=::NewEnumerator2
    Surface Only=false
    UseOwnerScale=true
   )
   &AddVelocity(Velocity=&MaskVectorBySpawnGroup(Masked Vector=&RandomRangeVector(Minimum=<0 -25 0>, Maximum=<0 -75 -5>), Passthrough Vector=<0 0 0>))
  ]
  particleUpdateScript [
   &ParticleState()
   &ScaleSpriteSize(
    Scale Sprite Size Mode=::NewEnumerator0
    Uniform Scale Factor=&MaskFloatBySpawnGroup(
     Masked Float=1
     Passthrough Float=&FloatFromCurve(FloatCurve=DataInterface() {propertyValues {Curve {keys [{arriveTangent=2.204, interpMode=RCIM_Cubic, leaveTangent=2.204, tangentMode=RCTM_User, time=0, value=0.3}, <at 1 1 RCIM_Cubic>]}}})
    )
   )
   &ScaleColor(
    Scale RGB=&MakeVectorFromLinearColorRGB(Color=&MultiplyLinearColor(
      Linear Color=$User.VariantColor
      Scale Factor=&MakeVector4_FromFloat(Float=&MaskFloatBySpawnGroup(
        Masked Float=&FloatFromCurve(FloatCurve=DataInterface() {propertyValues {Curve {keys [<at 0 3>,<at 0.1 1>]}}})
        Passthrough Float=1
       ))
     ))
    ScaleA=true
    ScaleRGB=true
   )
   &SolveForcesAndVelocity(Clamp Velocity=false, Limit Acceleration=false)
   &DynamicMaterialParameters(
    Index 0 Param 1=&RandomRangeFloat(Minimum=2.5, Maximum=3.5)
    Index 0 Param 2=&FloatFromCurve(FloatCurve=DataInterface() {propertyValues {Curve {keys [<at 0 1>,<at 1 -1>]}}})
    Index 0 Param 3=&RandomRangeFloat(Minimum=0.2, Maximum=0.7)
    Module.Param0WriteEnabled=true
    Module.Param1WriteEnabled=true
    Module.Param2WriteEnabled=true
    Module.Param3WriteEnabled=true
   )
   /SetVariables(Particles.New Linear Color:LinearColor=&MultiplyLinearColor(
     Linear Color=$User.VariantColor
     Scale Factor=&MakeVector4_FromFloat(Float=&FloatFromCurve(FloatCurve=DataInterface() {propertyValues {Curve {keys [<at 0 0>,<at 0.1 1>]}}}))
    )) {isSetParameters=true}
  ]
  rendererClasses [&NiagaraSpriteRendererProperties, &NiagaraLightRendererProperties]
  renderers [
   {rendererClass=&NiagaraSpriteRendererProperties}
   {rendererClass=&NiagaraSpriteRendererProperties, rendererIndex=1}
   {rendererClass=&NiagaraLightRendererProperties, rendererIndex=2}
  ]
 }}
summary {
 systemName=NS_Drift1H_Pickaxe_Idle
 emitters [{emitterName=Idle, rendererClasses [&NiagaraSpriteRendererProperties, &NiagaraLightRendererProperties]}]
 userVariables [User.VariantColor {type=LinearColor, description="", defaultValue {struct=LinearColor, value=<rgba 2 0 0.3 1>}}]
}
