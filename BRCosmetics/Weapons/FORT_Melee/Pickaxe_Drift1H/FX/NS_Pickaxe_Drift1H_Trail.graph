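# Niagara system for the melee swing trail of the one-handed Drift pickaxe, pairing animation-driven ribbons with sparks.
# Trail_Main (near-white) and shorter-lived Trail_OnTop (hot pink) build ribbons between two bone or socket indices, with randomized material parameters.
# GPU Sprite_Sparks spawns pink sparks with tapering probability after a short delay, moved by point force, downward acceleration and tiled vector-field turbulence.
# All three emitters fade over lifetime, with alpha scaled by time of day.
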NiagaraSystem /BRCosmetics/Weapons/FORT_Melee/Pickaxe_Drift1H/FX/NS_Pickaxe_Drift1H_Trail
dependencies {
 usedDataInterfaces [
  &NiagaraDataInterfaceVectorCurve
  &NiagaraDataInterfaceCurve
  &NiagaraDataInterfaceColorCurve
  /Script/Niagara.NiagaraDataInterfaceVectorField
  &NiagaraDataInterfaceSocketReader
 ]
 usedDynamicInputs [&RandomRangeVector, &Vector4FromCurve]
 usedModules [
  &self:SystemSpawnScript.SystemScriptSource.SystemScriptGraph.NiagaraNodeAssignment_0.SetVariables_#1
  &SystemState
 ]
}
emitters {
 Sprite_Sparks {
  simTarget=GPUComputeSim
  emitterUpdateScript [
   &EmitterState(Life Cycle Mode=::NewEnumerator1, Loop Behavior=::NewEnumerator1, Loop Delay=0.06, Loop Duration=0.15, UseLoopDelay=true)
   &SpawnRate(
    Spawn Probability=&FloatFromCurve(FloatCurve=DataInterface() {propertyValues {Curve {keys [<at 0 1>,<at 1 0>]}}}, CurveIndex=$Emitter.NormalizedLoopAge, Scale Curve=3)
    SpawnRate=20
    Use Spawn Probability=true
   )
  ]
  particleSpawnScript [
   &InitializeParticle(
    Color=<rgba 5 0 1.25 0.2>
    Lifetime Max=0.65
    Lifetime Min=0.1
    Lifetime Mode=::NewEnumerator1
    Sprite Size Max=<2 3>
    Sprite Size Min=<0.5 2>
    Sprite Size Mode=::NewEnumerator2
    UsePositionOffset=false
    Write Material Random=true
   )
   &ShapeLocation(MaskSpawnGroup=false, Offset=<0 0 25>, Offset Mode=::NewEnumerator0, Override Seed=false, Sphere Radius=15, UseOwnerScale=true)
   &AddVelocity(
    Cone Angle=&RandomRangeFloat(Minimum=0, Maximum=0.06)
    Cone Axis=<1 1 0>
    Module.UseConeFalloff=false
    Module.UseVelDistribution=false
    Override Seed=false
    Velocity Mode=::NewEnumerator2
    Velocity Speed=&RandomRangeFloat(Minimum=150, Maximum=250)
   )
   &InheritVelocity(Inherited Velocity Amount Scale=<0.1 0.1 0.1>, Speed Limit=false, UseSpeedThreshold=false)
  ]
  particleUpdateScript [
   &ParticleState()
   &PointForce(
    Force Origin=&AddVectorToPosition(Position=/Niagara/DynamicInputs/Helpers/SimulationPosition(), Vector=<250 0 0>)
    Force Strength=800
    Module.Use Falloff=false
    Module.UseLinearForce=false
   )
   &ScaleColor(
    Scale Alpha=&FloatFromCurve(
     FloatCurve=DataInterface() {propertyValues {Curve {keys [<at 0.5 1>,<at 1 0>]}}}
     Scale Curve=&FNTimeOfDayScalar(Night Scale=0.8, Morning Scale=1, Evening Scale=1)
    )
    Scale RGB=&VectorFromCurve(VectorCurve=DataInterface() {propertyValues {
       XCurve {keys [<at 0 150 RCIM_Cubic>,<at 1 25 RCIM_Cubic>]}
       YCurve {keys [<at 0 150 RCIM_Cubic>,<at 1 25 RCIM_Cubic>]}
       ZCurve {keys [<at 0 150 RCIM_Cubic>, {arriveTangent=0.0253, interpMode=RCIM_Cubic, leaveTangent=0.0253, tangentMode=RCTM_User, time=1, value=25}]}
      }})
    ScaleA=true
    ScaleRGB=true
   )
   &AccelerationForce(Acceleration=<0 0 -800>)
   &SolveForcesAndVelocity(Clamp Velocity=false, Limit Acceleration=false)
   &ScaleSpriteSizeBySpeed(Max Scale Factor=<100 100>, Velocity Threshold=300000)
   /Niagara/VectorFields/SampleVectorField(
    FieldScale:Vector3f=<5 5 5>
    Sampled Vector Scale=50
    VectorField=DataInterface() {propertyValues {Field=/Game/Effects/VectorFields/Noisy_Turbulence_103, bTileX=true, bTileY=true, bTileZ=true}}
   )
   /Niagara/VectorFields/ApplyVectorField()
  ]
  rendererClasses [&NiagaraSpriteRendererProperties]
  renderers [{rendererClass=&NiagaraSpriteRendererProperties}]
 }
 Trail_Main {
  emitterSpawnScript [&FN_RibbonIDOffsetUtility()]
  emitterUpdateScript [
   &EmitterState()
   &FN_SpawnAnimTrail()
  ]
  particleSpawnScript [
   &FN_InitializeAnimTrailBoneIndecies(First Bone or Socket Index=1, Second Bone or Socket Index=0)
   &FN_InitializeAnimTrailBoneIndecies(AnimTrail Selection=1, First Bone or Socket Index=2, Second Bone or Socket Index=3) {enabled=false, module=FN_InitializeAnimTrailBoneIndecies001}
   &FN_InitializeAnimTrail(Color=<rgba 15 14.8 14.3 1>, Lifetime=0.25)
   &AddVelocity(Velocity=<0 -400 0>) {enabled=false}
   &DynamicMaterialParameters(
    Index 0 Param 1=&RandomRangeFloat(Minimum=0, Maximum=1)
    Index 0 Param 2=&RandomRangeFloat(Minimum=0, Maximum=1)
    Module.Index0Float0Write=true
    Module.Index0Float1Write=true
    Module.Index0Float2Write=true
    Module.Index0Float3Write=true
   )
  ]
  particleUpdateScript [
   &SolveForcesAndVelocity(Clamp Velocity=false, Limit Acceleration=false)
   &FN_UpdateAnimTrail() {module=FN_UpdateAnimTrail001}
   &ScaleColor(
    Scale Alpha=&FloatFromCurve(
     FloatCurve=DataInterface() {propertyValues {Curve {keys [
         {arriveTangent=-0.0536, interpMode=RCIM_Cubic, leaveTangent=-0.0536, tangentMode=RCTM_User, time=0, value=0.93}
         {arriveTangent=-0.3259, interpMode=RCIM_Cubic, leaveTangent=-0.3259, tangentMode=RCTM_User, time=0.08, value=0.92}
         {arriveTangent=-0.4729, interpMode=RCIM_Cubic, leaveTangent=-0.4729, tangentMode=RCTM_User, time=0.42, value=0.13}
         {arriveTangent=-0.0536, interpMode=RCIM_Cubic, leaveTangent=-0.0536, tangentMode=RCTM_User, time=1, value=0}
        ]}}}
     Scale Curve=&FNTimeOfDayScalar(Night Scale=0.8, Morning Scale=1, Evening Scale=1)
    )
    Scale RGB=&VectorFromCurve(VectorCurve=DataInterface() {propertyValues {
       XCurve {keys [
         <at 0 4.91 RCIM_Cubic>
         {arriveTangent=-0.6738, interpMode=RCIM_Cubic, leaveTangent=-0.6738, tangentMode=RCTM_User, time=0.11, value=0.51}
         <at 1 0.05 RCIM_Cubic>
        ]}
       YCurve {keys [
         <at 0 5 RCIM_Cubic>
         {arriveTangent=-0.7829, interpMode=RCIM_Cubic, leaveTangent=-0.7829, tangentMode=RCTM_User, time=0.11, value=0.47}
         <at 1 0 RCIM_Cubic>
        ]}
       ZCurve {keys [
         <at 0 3.79 RCIM_Cubic>
         {arriveTangent=0.197, interpMode=RCIM_Cubic, leaveTangent=0.197, tangentMode=RCTM_User, time=0.11, value=0.32}
         <at 1.0016 0.02 RCIM_Cubic>
        ]}
      }})
    ScaleA=true
    ScaleRGB=true
   )
  ]
  rendererClasses [&NiagaraRibbonRendererProperties]
  renderers [
   {rendererClass=&NiagaraRibbonRendererProperties}
   {rendererClass=&NiagaraRibbonRendererProperties, rendererIndex=1}
  ]
 }
 Trail_OnTop {
  emitterSpawnScript [&FN_RibbonIDOffsetUtility()]
  emitterUpdateScript [
   &EmitterState()
   &FN_SpawnAnimTrail()
  ]
  particleSpawnScript [
   &FN_InitializeAnimTrailBoneIndecies(First Bone or Socket Index=1, Second Bone or Socket Index=0)
   &FN_InitializeAnimTrailBoneIndecies(AnimTrail Selection=1, First Bone or Socket Index=2, Second Bone or Socket Index=3) {enabled=false, module=FN_InitializeAnimTrailBoneIndecies001}
   &FN_InitializeAnimTrail(Color=<rgba 2 0 0.5 1>, Lifetime=0.15)
   &AddVelocity(Velocity=<0 -400 0>) {enabled=false}
   &DynamicMaterialParameters(
    Index 0 Param 1=&RandomRangeFloat(Minimum=0, Maximum=1)
    Index 0 Param 2=&RandomRangeFloat(Minimum=0, Maximum=1)
    Module.Index0Float0Write=true
    Module.Index0Float1Write=true
    Module.Index0Float2Write=true
    Module.Index0Float3Write=true
   )
  ]
  particleUpdateScript [
   &SolveForcesAndVelocity(Clamp Velocity=false, Limit Acceleration=false)
   &FN_UpdateAnimTrail() {module=FN_UpdateAnimTrail001}
   &ScaleColor(
    Scale Alpha=&FloatFromCurve(
     FloatCurve=DataInterface() {propertyValues {Curve {keys [
         {arriveTangent=0.0744, interpMode=RCIM_Cubic, leaveTangent=0.0744, tangentMode=RCTM_User, time=0.5, value=1}
         {arriveTangent=-0.0304, interpMode=RCIM_Cubic, leaveTangent=-0.0304, tangentMode=RCTM_User, time=1, value=0}
        ]}}}
     Scale Curve=&FNTimeOfDayScalar(Night Scale=0.8, Morning Scale=1, Evening Scale=1)
    )
    Scale RGB=&VectorFromFloat(Value=&FloatFromCurve(FloatCurve=DataInterface() {propertyValues {Curve {keys [<at 0 25>,<at 1 10>]}}}))
    ScaleA=true
    ScaleRGB=true
   )
  ]
  rendererClasses [&NiagaraRibbonRendererProperties]
  renderers [
   {rendererClass=&NiagaraRibbonRendererProperties}
   {rendererClass=&NiagaraRibbonRendererProperties, rendererIndex=1}
  ]
 }
}
summary {
 systemName=NS_Pickaxe_Drift1H_Trail
 emitters [
  {emitterName=Trail_Main, rendererClasses [&NiagaraRibbonRendererProperties]}
  {emitterName=Trail_OnTop, rendererClasses [&NiagaraRibbonRendererProperties]}
  {emitterName=Sprite_Sparks, simTarget=GPUComputeSim, rendererClasses [&NiagaraSpriteRendererProperties]}
 ]
 userVariables [
  User.bIsFrontEnd {type=NiagaraBool}
  User.bIsFrontEndPreview {type=NiagaraBool}
  User.RibbonIDOffset {type=NiagaraInt32}
  User.SkeletalMesh {
   type=NiagaraDataInterfaceSkeletalMesh
   defaultValue {
    struct=NiagaraExt_VariableValue_DataInterface
    value {dataInterface=&self:NiagaraDataInterfaceSkeletalMesh_0}
   }
  }
 ]
}
