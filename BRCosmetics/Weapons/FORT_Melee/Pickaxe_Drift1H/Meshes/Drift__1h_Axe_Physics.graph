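# Physics asset for the Drift one-handed pickaxe, giving the axe collision bodies and a short jointed chain that can swing from it.
# Two boxes, one angled, cover the handle bone, and each of the three chain bones has a matching box, the middle one turned a quarter turn.
# Angle-limited constraints link the chain in series from the handle, with tight twist and a narrower first swing limit at the root joint.
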PhysicsAsset /BRCosmetics/Weapons/FORT_Melee/Pickaxe_Drift1H/Meshes/Drift__1h_Axe_Physics
bodies {
 Melee_TwoHanded_Handle [
  {shapeType=Box, extentX=6.0196, extentY=4.3998, extentZ=31.2361, center=<-0.1901 -8.0939 21.1746>}
  {shapeType=Box, extentX=5.3129, extentY=3.2621, extentZ=10.9229, center=<2.66 -8.0939 4.3579>, rotation=<rot 44.9997 0 0>}
 ]
 dyn_chain_01 [{shapeType=Box, extentX=6.6454, extentY=5.0574, extentZ=2.7202, center=<2.6439 -0.0004 0.0001>}]
 dyn_chain_02 [{shapeType=Box, extentX=6.6454, extentY=5.0574, extentZ=2.7202, center=<2.6439 0 0.0001>, rotation=<rot 0 0 -89.9997>}]
 dyn_chain_03 [{shapeType=Box, extentX=6.6454, extentY=5.0574, extentZ=2.7202, center=<2.6439 0 0.0001>}]
}
constraints [
 {bone1Name=dyn_chain_01, bone2Name=Melee_TwoHanded_Handle, swing1LimitDegrees=30, swing2LimitDegrees=60, twistLimitDegrees=10}
 {bone1Name=dyn_chain_02, bone2Name=dyn_chain_01, swing1LimitDegrees=60, swing2LimitDegrees=60, twistLimitDegrees=10}
 {bone1Name=dyn_chain_03, bone2Name=dyn_chain_02, swing1LimitDegrees=60, swing2LimitDegrees=60, twistLimitDegrees=10}
]
